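# Translucent unlit material for the Deimos beacon link effect, drawing a wiggling, smoky energy strand.
# SplineThicken sets noise-varied width, scaled by the dynamic Expand parameter, while two scrolling sine waves add vertex wiggle that fades at both ends.
# Layered panning smoke blends two colors, plus an optional travelling pulse, and shapes opacity with edge, depth and camera fades.
# The UseWParticles switch selects particle-space wiggle and particle color and alpha.
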Material /Game/Effects/Fort_Effects/Materials/Deimos/M_Deimos_Beacon_Link
graph {
 properties {
  bUsedWithMeshParticles=true
  blendMode=BLEND_Translucent
  pixelDepthOffsetMode=PDOM_Legacy
  shadingModel=MSM_Unlit
  twoSided=true
  usedShadingModels=Unlit
 }
 nodes [
  n1 = DynamicParameter() {paramNames [Expand,Param2,Param3,Param4]}
  n2 = Multiply(A=@n1.Expand, B=ScalarParameter() {defaultValue=50, parameterName="WidthBase (S)", sliderMin=400})
  n3 = Multiply(A=ScalarParameter() {defaultValue=1600, parameterName="WidthTip (S)"}, B=@n1.Expand)
  n4 = TextureSample(UVs=Panner(Coordinate=TextureCoordinate() {uTiling=0.25, vTiling=2}) {speedX=0.0015, speedY=0.025}) {texture=&T_NoisePack_01}
  n5 = ComponentMask(None=TextureCoordinate()) {g=true}
  n6 = MaterialFunctionCall() {materialFunction=/Engine/Functions/Engine_MaterialFunctions02/WorldPositionOffset/ObjectScale}
  n7 = Multiply(A=Time(), B=ScalarParameter() {defaultValue=20, parameterName=WiggleSpeed})
  n8 = Sine(None=Add(A=Multiply(A=@n5, B=@n6.Scale X), B=Multiply(A=@n7) {constB=2})) {period=600}
  n9 = Constant3Vector() {constant=<rgba 0 0 5 0>}
  n10 = StaticSwitchParameter(
   True=Transform(None=@n9) {transformSourceType=TRANSFORMSOURCE_ParticleWorld}
   False=Transform(None=@n9) {transformSourceType=TRANSFORMSOURCE_Local}
  ) {parameterName=UseWParticles}
  n11 = ComponentMask(None=TextureCoordinate()) {g=true}
  n12 = MaterialFunctionCall() {materialFunction=&ObjectScale}
  n13 = Sine(None=Add(A=Multiply(A=@n11, B=@n12.Scale X), B=Multiply(A=@n7))) {period=350}
  n14 = Constant3Vector() {constant=<rgba 0 2 2 0>}
  n15 = StaticSwitchParameter(
   True=Transform(None=@n14) {transformSourceType=TRANSFORMSOURCE_ParticleWorld}
   False=Transform(None=@n14) {transformSourceType=TRANSFORMSOURCE_Local}
  ) {parameterName=UseWParticles}
  n16 = Multiply(A=Add(A=Multiply(A=@n8, B=@n10), B=Multiply(A=@n13, B=@n15)), B=ScalarParameter() {defaultValue=400, parameterName=WiggleAmount})
  n17 = MaterialFunctionCall() {materialFunction=&LinearGradient}
  n18 = Multiply(A=Multiply(A=@n17.VGradient, B=OneMinus(None=@n17.VGradient))) {constB=4}
  Add(A=Add(
    A=MaterialFunctionCall(WidthBase=@n2, WidthTip=@n3, UVs for Thickness=@n4.G) {materialFunction=&SplineThicken}.WorldPositionOffset
    B=Multiply(A=@n16, B=@n18)
   ))
  n19 = ParticleColor()
  n20 = TextureSample!(
   UVs=Panner(Coordinate=TextureCoordinate() {uTiling=0.5, vTiling=7}) {speedX=0.025, speedY=0.1}
   Tex
   Level=Constant() {r=3}
   Apply View MipBias
  ) {mipValueMode=TMVM_MipLevel, texture=&T_Inky_Smoke_Tile}
  n21 = Add(A=Panner(Coordinate=TextureCoordinate() {vTiling=3}) {speedY=0.2}, B=Multiply(A=@n20.G) {constB=0.1})
  n22 = Add(
   A=@n20.G
   B=TextureSample!(
    UVs=MaterialFunctionCall(XY=@n21) {materialFunction=&Swizzle}.YX
    Tex
    Apply View MipBias
   ) {samplerType=SAMPLERTYPE_Grayscale, texture=&T_Smoke_Trail_Soft}.G
  )
  n23 = Add(A=Multiply(A=@n22) {constB=0.05}, B=TextureCoordinate() {coordinateIndex=1, vTiling=0.5})
  n24 = Multiply(A=Time(), B=ScalarParameter() {defaultValue=0.2, parameterName=PulseSpeed})
  n25 = Multiply(
   A=TextureSample(UVs=Panner(Coordinate=@n23, Time=@n24) {speedY=1}) {texture=/Game/Effects/Fort_Effects/Textures/Mask/T_Rand_Pan_Mask}.G
   B=ScalarParameter() {parameterName=Pulse}
  )
  n26 = VectorParameter() {parameterName=Color2, defaultValue=<rgba 1.1416 0 2.5 1>}
  n27 = Multiply(A=@n26.RGB, B=@n26.RGB)
  n28 = Add(
   A=Multiply(A=Multiply(A=@n25, B=@n27))
   B=LinearInterpolate(
    A=VectorParameter() {parameterName=Color1, defaultValue=<rgba 0.065 0.0521 1 1>}.RGB
    B=@n26.RGB
    Alpha=Power(Base=@n22) {constExponent=1.25}
   )
  )
  MaterialFunctionCall(Emissive=StaticSwitchParameter(True=Multiply(A=@n19.RGB, B=@n28), False=@n28) {parameterName=UseWParticles}) {materialFunction=&MF_FX_EmissiveAdjustments}
  Multiply(A=@n25) {constB=2}
  n29 = ComponentMask(None=TextureCoordinate() {coordinateIndex=1}) {g=true}
  n30 = Time() {bOverride_Period=true, period=3}
  SphereMask(
   A=Add(A=Power(Base=@n29) {constExponent=0.5}, B=ConstantBiasScale(None=@n30) {bias=-0.5, scale=2})
   B=Constant() {r=0.5}
   Radius=Constant() {r=0.075}
   Hardness=Constant()
  )
  n31 = Add(A=@n22, B=Multiply(A=@n25) {constB=0.1})
  n32 = MaterialFunctionCall() {materialFunction=&LinearGradient}
  n33 = OneMinus(None=@n32.UGradient)
  n34 = OneMinus(None=@n32.VGradient)
  n35 = Multiply(A=Multiply(A=Multiply(A=@n32.UGradient, B=@n33)) {constB=4}, B=Multiply(A=Multiply(A=@n32.VGradient, B=@n34)) {constB=4})
  n36 = Multiply(
   A=DepthFade(Opacity=Multiply(A=@n31, B=@n35), FadeDistance=ScalarParameter() {defaultValue=1024, parameterName=FadeDistance})
   B=MaterialFunctionCall(
    Fade Length=ScalarParameter() {defaultValue=4000, parameterName=CameraFadeLength}
    Fade Offset=ScalarParameter() {defaultValue=4000, parameterName=FadeOffset}
   ) {materialFunction=&CameraDepthFade}.Result
  )
  StaticSwitchParameter(True=Multiply(A=@n19.A, B=@n36), False=@n36) {parameterName=UseWParticles}
 ]
}
