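# Additive, unlit, two-sided mesh particle material for a Deimos swirl effect on spline meshes, with glowing two-color smoky strands.
# Panning noise and inky smoke textures, offset by vertex red, blend the colors and add bright highlights.
# Opacity combines edge-faded smoke with a traveling pulse mask, with optional particle alpha scaling.
# SplineThicken shapes width with panning noise, while a Grow parameter slides a V-axis mask over thickness and smoke opacity.
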Material /Game/Effects/Fort_Effects/Materials/Deimos/M_Deimos_Swirl_Spline
graph {
 properties {
  bUsedWithMeshParticles=true
  blendMode=BLEND_Additive
  pixelDepthOffsetMode=PDOM_Legacy
  shadingModel=MSM_Unlit
  twoSided=true
  usedShadingModels=Unlit
 }
 nodes [
  n1 = Panner(Coordinate=Add(A=VertexColor().R, B=TextureCoordinate() {uTiling=2, vTiling=8})) {speedX=0.025, speedY=-0.15}
  n2 = VectorParameter() {parameterName=Color1, defaultValue=<rgba 0.065 0.0521 1 1>}
  n3 = Multiply(A=Power(Base=TextureSample(UVs=@n1) {texture=/Game/Effects/Fort_Effects/Textures/Tile/T_TileNoise_01}.G), B=@n2.RGB)
  n4 = VertexColor()
  n5 = TextureCoordinate() {uTiling=0.25, vTiling=2}
  n6 = TextureSample(UVs=Panner(Coordinate=Add(A=@n4.R, B=@n5)) {speedX=0.025, speedY=-0.15}) {texture=&T_Inky_Smoke_Tile}
  n7 = LinearInterpolate(A=@n2.RGB, B=VectorParameter() {parameterName=Color2, defaultValue=<rgba 2.3 0 3 1>}.RGB, Alpha=@n6.G)
  Add(A=Multiply(A=Multiply(A=@n3) {constB=1000}, B=@n6.G), B=@n7)
  n8 = Add(A=Multiply(A=@n6.G) {constB=0.1}, B=TextureCoordinate() {vTiling=0.5})
  n9 = Multiply(A=Add(A=Time()), B=ScalarParameter() {defaultValue=0.2, parameterName=PulseSpeed})
  n10 = Multiply(
   A=TextureSample(UVs=Panner(Coordinate=@n8, Time=@n9) {speedY=1}) {texture=/Game/Effects/Fort_Effects/Textures/Mask/T_Rand_Pan_Mask}.G
   B=ScalarParameter() {parameterName=Pulse}
  )
  n11 = MaterialFunctionCall() {materialFunction=&LinearGradient}
  n12 = Multiply(A=@n11.UGradient, B=OneMinus(None=@n11.UGradient))
  n13 = Multiply(A=@n11.VGradient, B=OneMinus(None=@n11.VGradient))
  n14 = Multiply(A=@n6.G, B=Multiply(A=Multiply(A=@n12) {constB=4}, B=Multiply(A=@n13) {constB=4}))
  n15 = ScalarParameter() {defaultValue=0.1, parameterName=Opacity}
  n16 = TextureCoordinate()
  n17 = ScalarParameter() {defaultValue=1, parameterName=Grow}
  n18 = Clamp(None=Add(A=ComponentMask(None=@n16) {g=true}, B=LinearInterpolate(Alpha=@n17) {constA=-1}))
  n19 = Add(A=Multiply(A=Multiply(A=@n10, B=@n7)) {constB=0.025}, B=Multiply(A=Multiply(A=@n14, B=@n15), B=@n18))
  LinearInterpolate(A=@n19, B=Multiply(A=@n19, B=ParticleColor().A), Alpha=ScalarParameter() {parameterName=ParticleColor})
  n20 = Panner(Coordinate=Add(A=VertexColor().R, B=TextureCoordinate() {uTiling=0.1, vTiling=2})) {speedX=0.0015, speedY=0.025}
  MaterialFunctionCall(
   WidthBase=ScalarParameter() {defaultValue=1000, parameterName="WidthBase (S)"}
   WidthTip=Multiply(A=ScalarParameter() {defaultValue=4000, parameterName="WidthTip (S)"}, B=@n18)
   UVs for Thickness=Multiply(A=TextureSample(UVs=@n20) {texture=&T_NoisePack_01}.G, B=@n18)
  ) {materialFunction=&SplineThicken}
  MaterialFunctionCall() {materialFunction=&DebugTimeSine}
 ]
}
