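# Unlit, two-sided, dither-masked material for the glowing center of the Deimos vortex effect, usable on mesh particles.
# Layered electric noise panned in radial coordinates distorts itself and, with a center sphere mask, blends Color1 to Color2.
# A noise-warped, scrolling mask adds brightness controlled by Pulse and PulseSpeed.
# A GrowCenter-sized sphere mask modulated by noise shapes the opacity, and UseWParticles optionally multiplies in particle color and alpha.
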Material /Game/Effects/Fort_Effects/Materials/Deimos/M_Deimos_Vortex_CenterGlow
graph {
 properties {
  bUsedWithMeshParticles=true
  blendMode=BLEND_Masked
  ditherOpacityMask=true
  pixelDepthOffsetMode=PDOM_Legacy
  shadingModel=MSM_Unlit
  twoSided=true
  usedShadingModels=Unlit
 }
 nodes [
  n1 = SphereMask(A=TextureCoordinate(), B=Constant2Vector() {g=0.5, r=0.5}, Radius=Constant() {r=0.5}, Hardness=Constant())
  Clamp(None=Power(Base=Multiply(A=@n1) {constB=1.15}) {constExponent=3})
  MaterialFunctionCall() {materialFunction=&DebugTimeSine}
  MaterialFunctionCall!(UVs) {materialFunction=/Game/Effects/Fort_Effects/Materials/_Functions/VectortoRadialCorrection}
  n2 = ParticleColor()
  n3 = Panner(Coordinate=MaterialFunctionCall() {materialFunction=&VectorToRadialValue}.Radial Coordinates) {speedX=-0.015, speedY=-0.05}
  n4 = Constant2Vector() {g=1, r=2}
  n5 = MaterialFunctionCall() {materialFunction=&VectorToRadialValue}
  n6 = TextureSample(UVs=Multiply(A=Panner(Coordinate=@n5.Radial Coordinates) {speedX=-0.04, speedY=-0.05}, B=Constant2Vector() {g=2, r=3})) {samplerType=SAMPLERTYPE_LinearColor, texture=/Game/Characters/Player/Male/Medium/Bodies/M_MED_SRDrift/Textures/T_ElectricNoise}
  n7 = TextureSample(UVs=Add(A=Multiply(A=@n3, B=@n4), B=Multiply(A=@n6.G) {constB=0.05})) {samplerType=SAMPLERTYPE_LinearColor, texture=&T_ElectricNoise}
  n8 = Time()
  n9 = Panner(
   Coordinate=Add(
    A=Multiply(A=@n7.G) {constB=0.1}
    B=MaterialFunctionCall() {materialFunction=&VectorToRadialValue}.Radial Coordinates
   )
   Time=Multiply(A=Add(A=@n8), B=ScalarParameter() {defaultValue=0.2, parameterName=PulseSpeed})
  ) {speedY=-1}
  n10 = Constant2Vector() {g=0.5, r=1}
  n11 = Multiply(
   A=TextureSample(UVs=Multiply(A=@n9, B=@n10)) {texture=/Game/Effects/Fort_Effects/Textures/Mask/T_Rand_Pan_Mask}.G
   B=ScalarParameter() {parameterName=Pulse}
  )
  n12 = Power(Base=Multiply(A=@n1) {constB=1.1}) {constExponent=8}
  n13 = Add(A=@n7.G, B=LinearInterpolate(Alpha=@n6.G) {constB=0.2})
  n14 = LinearInterpolate(
   A=VectorParameter() {parameterName=Color1, defaultValue=<rgba 0.0305 0.0244 0.4688 1>}.RGB
   B=VectorParameter() {parameterName=Color2, defaultValue=<rgba 1.3699 0 3 1>}.RGB
   Alpha=Add(A=Clamp(None=@n12), B=@n13)
  )
  n15 = Add(A=Multiply(A=Multiply(A=@n11, B=@n14)) {constB=0.5}, B=@n14)
  MaterialFunctionCall(Emissive=StaticSwitchParameter(True=Multiply(A=@n2.RGB, B=@n15), False=@n15) {parameterName=UseWParticles}) {materialFunction=&MF_FX_EmissiveAdjustments}
  n16 = Constant() {r=0.5}
  n17 = ScalarParameter() {defaultValue=1, parameterName=GrowCenter}
  n18 = Multiply(
   A=LinearInterpolate(Alpha=@n13) {constA=0.75}
   B=SphereMask(A=TextureCoordinate(), B=Constant2Vector() {g=0.5, r=0.5}, Radius=LinearInterpolate(B=@n16, Alpha=@n17), Hardness=Constant() {r=0.5})
  )
  StaticSwitchParameter(True=Multiply(A=@n2.A, B=@n18), False=@n18) {parameterName=UseWParticles}
  TextureCoordinate() {vTiling=0.5}
 ]
}
