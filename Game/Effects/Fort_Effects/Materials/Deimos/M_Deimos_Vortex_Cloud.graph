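# Translucent material for a swirling Deimos vortex cloud, built for use on mesh particles, including Niagara ones.
# Panning smoke samples, skewed along the V gradient, blend two colors for emissive, with a scrolling mask adding a tunable pulse.
# A noise-warped gradient band masks opacity and normal-direction displacement, which samples a low smoke mip for smoother shapes.
# Grow erodes opacity; switches add Fresnel edge fade or let particle color and dynamic parameters drive the look.
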Material /Game/Effects/Fort_Effects/Materials/Deimos/M_Deimos_Vortex_Cloud
graph {
 properties {
  bUsedWithMeshParticles=true
  bUsedWithNiagaraMeshParticles=true
  blendMode=BLEND_Translucent
  pixelDepthOffsetMode=PDOM_Legacy
  usedShadingModels=Default Lit
 }
 nodes [
  MaterialFunctionCall() {materialFunction=&DebugTimeSine}
  n1 = ParticleColor()
  n2 = Multiply(
   A=Power(Base=MaterialFunctionCall() {materialFunction=&LinearGradient}.VGradient)
   B=Constant2Vector() {r=1}
  )
  n3 = Multiply(A=Time(), B=ScalarParameter() {defaultValue=1, parameterName=Speed})
  n4 = Panner(Coordinate=Add(A=@n2, B=TextureCoordinate() {vTiling=0.5}), Time=@n3) {speedX=0.025, speedY=-0.025}
  n5 = Add(A=@n2, B=TextureCoordinate() {uTiling=2})
  n6 = Multiply(
   A=TextureSample(UVs=@n4) {texture=/Game/Global/Textures/Default/Tileable_Patterns/T_Smoke_Puff_Tile_Up}.G
   B=TextureSample(UVs=Panner(Coordinate=@n5, Time=@n3) {speedX=0.025, speedY=-0.025}) {texture=&T_Smoke_Puff_Tile_Up}.G
  )
  n7 = LinearInterpolate(
   A=VectorParameter() {parameterName=ColorA, defaultValue=<rgba 0.0237 0.0206 0.253 1>}.RGB
   B=VectorParameter() {parameterName=ColorB, defaultValue=<rgba 0.7521 0.1236 1.5 1>}.RGB
   Alpha=@n6
  )
  n8 = Multiply(A=@n6) {constB=0.15}
  n9 = TextureCoordinate() {vTiling=0.5}
  n10 = Add(A=Time())
  n11 = ScalarParameter() {defaultValue=0.2, parameterName=PulseSpeed}
  n12 = TextureSample(UVs=Panner(Coordinate=Add(A=@n8, B=@n9), Time=Multiply(A=@n10, B=@n11)) {speedY=1}) {texture=/Game/Effects/Fort_Effects/Textures/Mask/T_Rand_Pan_Mask}
  n13 = ScalarParameter() {parameterName=Pulse}
  n14 = Multiply(A=Multiply(A=Multiply(A=@n12.G, B=@n13), B=@n7))
  n15 = Power(Base=MaterialFunctionCall() {materialFunction=&LinearGradient}.VGradient)
  n16 = Multiply(A=@n7, B=@n15)
  n17 = ScalarParameter() {defaultValue=1, parameterName=EmissiveAmount}
  n18 = Add(A=Multiply(A=@n7) {constB=0.1}, B=Add(A=Multiply(A=@n14, B=@n15), B=Multiply(A=@n16, B=@n17) {constB=0.025}))
  MaterialFunctionCall(Emissive=StaticSwitchParameter(True=Multiply(A=@n1.RGB, B=@n18), False=@n18) {parameterName=UseWParticles}) {materialFunction=&MF_FX_EmissiveAdjustments}
  n19 = Multiply(
   A=TextureSample!(UVs=@n4, Tex, Level=Constant() {r=3}, Apply View MipBias) {mipValueMode=TMVM_MipLevel, texture=&T_Smoke_Puff_Tile_Up}.G
   B=VertexNormalWS()
  )
  n20 = StaticSwitchParameter(
   True=DynamicParameter() {paramNames [Grow, DisplaceAmount, Param3, Param4]}.DisplaceAmount
   False=ScalarParameter() {defaultValue=1000, parameterName=DisplaceAmount}
  ) {parameterName=UseWParticles}
  n21 = MaterialFunctionCall(UV Channel=Add(A=Multiply(A=@n6) {constB=0.075}, B=TextureCoordinate())) {materialFunction=&LinearGradient}
  n22 = Multiply(A=OneMinus(None=@n21.VGradient)) {constB=8}
  n23 = Multiply(A=@n21.VGradient, B=Clamp(None=Power(Base=@n22) {constExponent=1.5}))
  n24 = Clamp(None=Power(Base=Multiply(A=@n23) {constB=2}) {constExponent=2.5})
  Reroute(None=Multiply(A=@n24, B=Multiply(A=@n19, B=@n20)))
  n25 = Fresnel(ExponentIn=ScalarParameter() {defaultValue=5, parameterName=Fresnel Expo})
  n26 = TextureCoordinate() {uTiling=4, vTiling=2}
  n27 = TextureSample!(UVs=Panner(Coordinate=Add(A=@n2, B=@n26), Time=@n3) {speedX=0.025, speedY=-0.025}, Tex, Apply View MipBias) {texture=&T_Smoke_Puff_Tile_Up}
  n28 = Clamp(None=Power(Base=Multiply(A=@n27.G) {constB=1.2}) {constExponent=0.25})
  n29 = TextureCoordinate()
  n30 = Multiply(
   A=LinearInterpolate(Alpha=@n6) {constA=0.5}
   B=Power(Base=ComponentMask(None=@n29) {g=true}, Exp=ScalarParameter() {defaultValue=1, parameterName=Edge Fade Contrast}) {constExponent=1}
  )
  n31 = LinearInterpolate(Alpha=StaticSwitchParameter(
    True=DynamicParameter() {paramNames [Grow, DisplaceAmount, Param3, Param4]}.Grow
    False=ScalarParameter() {defaultValue=1, parameterName=Grow}
   ) {parameterName=UseWParticles}) {constA=-1}
  n32 = Multiply(A=Multiply(A=ScalarParameter() {defaultValue=1, parameterName=Opacity}, B=Multiply(A=@n24, B=@n28)), B=Clamp(None=Add(A=@n30, B=@n31)))
  n33 = StaticSwitchParameter(True=Multiply(A=@n32, B=OneMinus(None=@n25)), False=@n32) {parameterName=Add Fresnel Based Edge Fade}
  StaticSwitchParameter(True=Multiply(A=@n1.A, B=@n33), False=@n33) {parameterName=UseWParticles}
  TextureSample() {texture=&AtmosphericCloudNoise01}
  TextureSample() {samplerType=SAMPLERTYPE_Normal, texture=/Game/Global/Textures/Default/Tileable_Patterns/T_StormCloudDetail01_N}
  TextureSample() {samplerType=SAMPLERTYPE_Normal, texture=&T_StormCloudDetail01_N}
 ]
}
